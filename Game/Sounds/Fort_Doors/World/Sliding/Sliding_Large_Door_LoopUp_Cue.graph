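# Sound cue for the looping movement of a large sliding door, on the Fort_Doors sound class with no-occlusion door attenuation.
# A mixer layers a looping military base door sound, a looping slow metal slide, and a randomly delayed, pitched-down scrape picked from three non-looping waves.
# The mix passes through a short fixed delay into a root Enveloper that fades the volume in over about half a second.
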SoundCue /Game/Sounds/Fort_Doors/World/Sliding/Sliding_Large_Door_LoopUp_Cue
root=Enveloper_1
nodes [
 {id=Delay_1, children [Mixer_0], properties {delayMax=0.1, delayMin=0.1}}
 {id=Delay_2, children [Modulator_0], properties {delayMax=0.5, delayMin=0.1}}
 {
  id=Enveloper_1
  children [Delay_1]
  properties {pitchCurve {editorCurveData {keys [{time=0}]}}, volumeCurve {editorCurveData {keys [<at -0.0054 0>,<at 0.2662 0.3606>,<at 0.5 1>]}}}
 }
 {id=Mixer_0, children [WavePlayer_0,WavePlayer_1,Delay_2], properties {inputVolume [0.95,1.5,0.7]}}
 {id=Modulator_0, children [Random_0], properties {pitchMax=0.7, pitchMin=0.6, volumeMin=1}}
 {id=Random_0, children [WavePlayer_8,WavePlayer_5,WavePlayer_6], properties {weights [1,1,1]}}
 {id=WavePlayer_0, properties {bLooping=true, soundWaveAssetPtr=/Game/Sounds/Creative/Toys/Doors/MilitaryBase_Door_LoopUp01}}
 {
  id=WavePlayer_1
  properties {bLooping=true, soundWaveAssetPtr=/Game/Sounds/Physics/Objects/Metal/MetalLarge/Slides/Phys_Metal_Large_Slide_Slow_Loop}
 }
 {id=WavePlayer_5, properties {soundWaveAssetPtr=/Game/Sounds/Trans_AmbientNonLoop/Metal/ScrapeLongA04}}
 {id=WavePlayer_6, properties {soundWaveAssetPtr=/Game/Sounds/Trans_AmbientNonLoop/Metal/ScrapeLongA03}}
 {id=WavePlayer_8, properties {soundWaveAssetPtr=/Game/Sounds/Trans_AmbientNonLoop/Metal/ScrapeLongA02}}
]
properties {
 attenuationSettings=/Game/Sounds/Attenuation/Fort_Doors_NoOcclusion_Att
 maxDistance=2400
 pitchMultiplier=0.9
 soundClassObject=&Fort_Doors
 volumeMultiplier=0.8
 concurrencyOverrides {resolutionRule=PreventNew}
}
