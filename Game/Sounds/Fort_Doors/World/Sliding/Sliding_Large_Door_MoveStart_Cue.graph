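# Sound cue that plays the start-of-movement sound for a large sliding door, reusing a military base door wave.
# A single wave player feeds a Modulator for variation, with the cue's pitch and volume slightly lowered.
# It uses the door sound class and a no-occlusion door attenuation.
# A shared "only one" military door concurrency set limits overlapping instances.
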SoundCue /Game/Sounds/Fort_Doors/World/Sliding/Sliding_Large_Door_MoveStart_Cue
root=Modulator_0
nodes [
 {id=Modulator_0, children [WavePlayer_0]}
 {id=WavePlayer_0, properties {soundWaveAssetPtr=/Game/Sounds/Creative/Toys/Doors/MilitaryBase_Door_MoveStart01}}
]
properties {
 attenuationSettings=/Game/Sounds/Attenuation/Fort_Doors_NoOcclusion_Att
 maxDistance=2400
 pitchMultiplier=0.9
 soundClassObject=&Fort_Doors
 volumeMultiplier=0.85
 concurrencyOverrides {resolutionRule=PreventNew}
 concurrencySet [/Game/Sounds/Creative/Concurrency/MilitaryDoor_OnlyOne_Conc]
}
